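# Additive, unlit, two-sided Niagara sprite material for an off-pawn trail effect on the HeatShineTorn cosmetic shoe.
# A view-facing term that fades grazing angles multiplies a V-axis side falloff and a U-axis band with adjustable base and tail fades.
# That mask feeds a depth fade whose distance differs between front and back faces.
# Particle color passes through a shared emissive adjustment function, and a CameraOffset function supplies an adjustable camera-relative position offset.
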Material /CosmeticShoes/Assets/FORT_Shoes/Shoe_HeatShineTorn/FX/Materials/M_Shoe_HeatShineTorn_OffPawnTrail
graph {
 properties {
  bAutomaticallySetUsageInEditor=false
  bUsedWithNiagaraSprites=true
  bUsedWithStaticMesh=false
  blendMode=BLEND_Additive
  shadingModel=MSM_Unlit
  twoSided=true
  usedShadingModels=Unlit
 }
 nodes [
  n1 = Abs(None=DotProduct(A=VertexNormalWS(), B=CameraVectorWS()))
  n2 = ScalarParameter() {defaultValue=4, parameterName=Fresnel Exponent}
  n3 = TextureCoordinate()
  n4 = ConstantBiasScale(None=Saturate(None=ComponentMask(None=@n3) {g=true})) {bias=-0.5, scale=2}
  n5 = Power(Base=OneMinus(None=Abs(None=@n4)), Exp=ScalarParameter() {defaultValue=1, parameterName=Trail Side Falloff Exponent})
  n6 = ComponentMask(None=@n3) {r=true}
  n7 = ScalarParameter() {defaultValue=0.16, parameterName=Blue Band Base Fade Size}
  n8 = Divide(A=@n6, B=@n7)
  n9 = ScalarParameter() {defaultValue=2, parameterName=Blue Band Tail Fade Exponent}
  n10 = Saturate(None=Divide(A=Subtract(A=@n6, B=@n7), B=OneMinus(None=@n7)))
  n11 = Multiply(A=Power(Base=Saturate(None=@n8), Exp=@n9), B=Power(Base=OneMinus(None=@n10), Exp=@n9))
  n12 = TwoSidedSign()
  DepthFade(
   Opacity=Multiply(A=Power(Base=@n1, Exp=@n2), B=Multiply(A=@n5, B=@n11))
   FadeDistance=LinearInterpolate(
    A=ScalarParameter() {defaultValue=44, parameterName=Depth Fade Distance Backside}
    B=ScalarParameter() {defaultValue=5, parameterName=Depth Fade Distance}
    Alpha=Saturate(None=@n12)
   )
  )
  MaterialFunctionCall(Offset Amount=ScalarParameter() {defaultValue=11, parameterName=Camera Offset Amount}) {materialFunction=/Engine/Functions/Engine_MaterialFunctions02/WorldPositionOffset/CameraOffset}
  MaterialFunctionCall(Emissive=ParticleColor().RGB) {materialFunction=&MF_FX_EmissiveAdjustments}
 ]
}
