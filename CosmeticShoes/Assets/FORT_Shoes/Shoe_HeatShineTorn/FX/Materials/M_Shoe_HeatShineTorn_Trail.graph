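# Additive unlit Niagara ribbon material for the trail of the HeatShineTorn cosmetic shoe.
# It splits the ribbon's V axis into a tinted blue band and a color band sweeping red to green via HSV, using a curved hue gradient.
# Both bands fade at their edges, base and tail; the masked blend is scaled into MF_FX_EmissiveAdjustments.
# A view-angle Fresnel fade term is also computed.
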Material /CosmeticShoes/Assets/FORT_Shoes/Shoe_HeatShineTorn/FX/Materials/M_Shoe_HeatShineTorn_Trail
graph {
 properties {
  bAutomaticallySetUsageInEditor=false
  bUsedWithNiagaraRibbons=true
  bUsedWithStaticMesh=false
  blendMode=BLEND_Additive
  shadingModel=MSM_Unlit
  usedShadingModels=Unlit
 }
 nodes [
  n1 = NamedRerouteUsage()
  n2 = ComponentMask(None=TextureCoordinate()) {g=true}
  n3 = ScalarParameter() {defaultValue=0.5, parameterName=Blue Band Relative Size}
  n4 = Subtract(A=@n2, B=@n3)
  n5 = OneMinus(None=@n3)
  n6 = NamedRerouteDeclaration(None=Saturate(None=Divide(A=@n4, B=@n5))) {name=Color Band Gradient, nodeColor=<rgba 0 1 0.5547 1>}
  n7 = ConstantBiasScale(None=@n6) {bias=-0.5, scale=2}
  n8 = Abs(None=@n7)
  n9 = OneMinus(None=Power(Base=OneMinus(None=@n8), Exp=ScalarParameter() {defaultValue=4, parameterName=Color Band Curve Exponent}))
  n10 = Sign(None=@n7)
  n11 = LinearInterpolate(Alpha=ConstantBiasScale(None=Multiply(A=@n9, B=@n10))) {constB=0.333}
  n12 = OneMinus(None=Abs(None=ConstantBiasScale(None=@n6) {bias=-0.5, scale=2}))
  n13 = ScalarParameter() {defaultValue=2, parameterName=Color Band Edges Fade Exponent}
  n14 = Subtract(A=ComponentMask(None=TextureCoordinate()) {r=true}, B=ScalarParameter() {defaultValue=0.1, parameterName=Color Band Base Fade Offset})
  n15 = ScalarParameter() {defaultValue=0.1, parameterName=Color Band Base Fade Size}
  n16 = ScalarParameter() {defaultValue=3, parameterName=Color Band Tail Fade Exponent}
  n17 = Power(Base=Saturate(None=Divide(A=@n14, B=@n15)), Exp=@n16)
  n18 = Divide(A=Subtract(A=@n14, B=@n15), B=OneMinus(None=@n15))
  n19 = Power(Base=OneMinus(None=Saturate(None=@n18)), Exp=@n16)
  n20 = AppendVector(A=Constant() {r=1}, B=Multiply(A=Power(Base=@n12, Exp=@n13), B=Multiply(A=@n17, B=@n19)))
  n21 = NamedRerouteDeclaration(None=HsvToRgb!(None=AppendVector(A=@n11, B=@n20))) {name=Color Band Emissive Color, nodeColor=<rgba 0 0.7187 1 1>}
  n22 = NamedRerouteUsage()
  MaterialFunctionCall(Emissive=Multiply(A=LinearInterpolate(A=@n1, B=@n21, Alpha=@n22), B=ScalarParameter() {defaultValue=0.55, parameterName=Emissive Multiplier})) {materialFunction=&MF_FX_EmissiveAdjustments}
  n23 = Divide(A=@n2, B=@n3)
  NamedRerouteDeclaration(None=Saturate(None=@n23)) {name=Blue Band Gradient, nodeColor=<rgba 0 0.4844 1 1>}
  NamedRerouteDeclaration(None=Saturate(None=Floor(None=@n23))) {name=Blue/Color Band Mask, nodeColor=<rgba 0 1 0.3906 1>}
  n24 = ConstantBiasScale(None=NamedRerouteUsage()) {bias=-0.5, scale=2}
  n25 = VectorParameter() {parameterName=Blue Band Color, defaultValue=<rgba 0 1.5 2 3>}
  n26 = Power(Base=OneMinus(None=Abs(None=@n24)), Exp=@n25.A)
  n27 = ComponentMask(None=TextureCoordinate()) {r=true}
  n28 = ScalarParameter() {defaultValue=0.16, parameterName=Blue Band Base Fade Size}
  n29 = ScalarParameter() {defaultValue=2, parameterName=Blue Band Tail Fade Exponent}
  n30 = Power(Base=Saturate(None=Divide(A=@n27, B=@n28)), Exp=@n29)
  n31 = Divide(A=Subtract(A=@n27, B=@n28), B=OneMinus(None=@n28))
  n32 = Power(Base=OneMinus(None=Saturate(None=@n31)), Exp=@n29)
  NamedRerouteDeclaration(None=Multiply(A=Multiply(A=@n26, B=@n25.RGB), B=Multiply(A=@n30, B=@n32))) {name=Blue Band Emissive Color, nodeColor=<rgba 1 0.0703 0 1>}
  n33 = DotProduct(A=VertexNormalWS(), B=CameraVectorWS())
  NamedRerouteDeclaration(None=Power(Base=Abs(None=@n33), Exp=ScalarParameter() {defaultValue=0.5, parameterName=Fresnel Exponent})) {name=Fresnel Fade, nodeColor=<rgba 0.1719 0 1 1>}
  NamedRerouteUsage()
 ]
}
