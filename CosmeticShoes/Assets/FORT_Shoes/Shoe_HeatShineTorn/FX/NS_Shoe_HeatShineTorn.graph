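# Niagara system for the HeatShineTorn cosmetic shoe, drawing trail effects at two sampled skeletal mesh points, left and right.
# FE_OffPawn places sprites at bone-relative offsets, with blue, green, red and yellow colors fading on a six-second curve.
# Ingame_New builds two short ribbons whose particles chase foot targets through InterpolateOverTime, slowing along each chain and as forward velocity rises, so trails stretch.
# Per-shoe scale sets ribbon width.
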NiagaraSystem /CosmeticShoes/Assets/FORT_Shoes/Shoe_HeatShineTorn/FX/NS_Shoe_HeatShineTorn
dependencies {
 usedDataInterfaces [&NiagaraDataInterfaceCurve]
 usedDynamicInputs [
  &Max_Float
  &TransformVector
  &MakeCustomFloatFromBool
 ]
 usedModules [
  &self:SystemSpawnScript.SystemScriptSource.SystemScriptGraph.NiagaraNodeAssignment_0.SetVariables_#1
  &SystemState
  &self:SystemSpawnScript.SystemScriptSource.SystemScriptGraph.NiagaraNodeAssignment_2.SetVariables_#2
  &self:SystemSpawnScript.SystemScriptSource.SystemScriptGraph.NiagaraNodeAssignment_1.SetVariables_#3
 ]
}
emitters {
 FE_OffPawn {
  emitterSpawnScript [
   /SetVariables(Emitter.SpriteAlignment:Vector3f=<-0.75 1 -1>, Emitter.SpriteFacing:Vector3f=<-1 -1 1>) {isSetParameters=true}
   /SetVariables(
    Emitter.BlueTrailColor:LinearColor=<rgba 0 0.8333 1 1>
    Emitter.BlueTrailSize:Vector2f=<33 2.5>
    Emitter.ColorTrailSize:Vector2f=<22 1.75>
    Emitter.GreenTrailColor:LinearColor=<rgba 0.08 0.4 0.08 1>
    Emitter.RedTrailColor:LinearColor=<rgba 0.8 0.08 0.14 1>
    Emitter.YellowTrailColor:LinearColor=<rgba 0.6 0.258 0.06 1>
   ) {isSetParameters=true}
  ]
  emitterUpdateScript [
   &EmitterState(Life Cycle Mode=::NewEnumerator1, Loop Duration=1, UseLoopDelay=false)
   /SetVariables(Emitter.FEOffPawnCycling:NiagaraFloat=&FloatFromCurve(
     FloatCurve=DataInterface() {propertyValues {Curve {
        defaultValue=1
        keys [
         {interpMode=RCIM_Cubic, tangentMode=RCTM_Break, time=0, value=0}
         {interpMode=RCIM_Cubic, tangentMode=RCTM_Break, time=0.25, value=1}
         {arriveTangentWeight=0.2778, interpMode=RCIM_Cubic, leaveTangentWeight=0.1389, tangentMode=RCTM_Break, time=0.5, value=1}
         {interpMode=RCIM_Cubic, tangentMode=RCTM_Break, time=0.75, value=0}
        ]
       }}}
     CurveIndex=&Frac_Float(Float=&Divide_Float(A=$Emitter.Age, B=6))
    )) {isSetParameters=true}
   &SampleSkeletalMesh(
    AdditionalYawPitchRoll=false
    Bone Sampling Mode:ENiagaraBoneSamplingMode=::NewEnumerator4
    Skeletal Mesh:NiagaraDataInterfaceSkeletalMesh=$System.SkeletalMesh
   )
   &SampleSkeletalMesh(
    AdditionalYawPitchRoll=false
    Bone / Socket Index=1
    Bone Sampling Mode:ENiagaraBoneSamplingMode=::NewEnumerator4
    Skeletal Mesh:NiagaraDataInterfaceSkeletalMesh=$System.SkeletalMesh
   ) {module=SampleSkeletalMesh001}
   /SetVariables(
    Emitter.BlueTrailPositionR:NiagaraPosition=&AddVectorToPosition(
     Position=$Emitter.SampleSkeletalMesh001.SampledPosition
     Vector=&RotateVectorByQuaternion(Vector To Rotate=<6.0039 -4.4564 4.9813>, Quaternion=$Emitter.SampleSkeletalMesh001.SampledRotation)
    )
    Emitter.ColorTrailPositionR:NiagaraPosition=&AddVectorToPosition(
     Position=$Emitter.SampleSkeletalMesh001.SampledPosition
     Vector=&RotateVectorByQuaternion(Vector To Rotate=<2.5903 3.1672 3.9835>, Quaternion=$Emitter.SampleSkeletalMesh001.SampledRotation)
    )
    Emitter.SpriteAlignmentR:Vector3f=&RotateVectorByQuaternion(Vector To Rotate=<1 0 0>, Quaternion=$Emitter.SampleSkeletalMesh001.SampledRotation)
    Emitter.SpriteFacingR:Vector3f=&RotateVectorByQuaternion(Vector To Rotate=<0 0.1 1>, Quaternion=$Emitter.SampleSkeletalMesh001.SampledRotation)
   ) {isSetParameters=true}
   /SetVariables(
    Emitter.BlueTrailPositionL:NiagaraPosition=&AddVectorToPosition(
     Position=$Emitter.SampleSkeletalMesh.SampledPosition
     Vector=&RotateVectorByQuaternion(Vector To Rotate=<-6.0039 4.4564 -4.9813>, Quaternion=$Emitter.SampleSkeletalMesh.SampledRotation)
    )
    Emitter.ColorTrailPositionL:NiagaraPosition=&AddVectorToPosition(
     Position=$Emitter.SampleSkeletalMesh.SampledPosition
     Vector=&RotateVectorByQuaternion(Vector To Rotate=<-2.5903 -3.1672 -3.9835>, Quaternion=$Emitter.SampleSkeletalMesh.SampledRotation)
    )
    Emitter.SpriteAlignmentL:Vector3f=&RotateVectorByQuaternion(Vector To Rotate=<1 0 0>, Quaternion=$Emitter.SampleSkeletalMesh.SampledRotation)
    Emitter.SpriteFacingL:Vector3f=&RotateVectorByQuaternion(Vector To Rotate=<0 0.1 -1>, Quaternion=$Emitter.SampleSkeletalMesh.SampledRotation)
   ) {isSetParameters=true}
   /SetVariables(
    Emitter.BlueTrailColor:LinearColor=&Multiply_LinearColorByFloat(Linear Color=$Emitter.Initial.BlueTrailColor, Float=$Emitter.FEOffPawnCycling)
    Emitter.GreenTrailColor:LinearColor=&Multiply_LinearColorByFloat(Linear Color=$Emitter.Initial.GreenTrailColor, Float=$Emitter.FEOffPawnCycling)
    Emitter.RedTrailColor:LinearColor=&Multiply_LinearColorByFloat(Linear Color=$Emitter.Initial.RedTrailColor, Float=$Emitter.FEOffPawnCycling)
    Emitter.YellowTrailColor:LinearColor=&Multiply_LinearColorByFloat(Linear Color=$Emitter.Initial.YellowTrailColor, Float=$Emitter.FEOffPawnCycling)
   ) {isSetParameters=true}
  ]
  rendererClasses [&NiagaraSpriteRendererProperties]
  renderers [
   {rendererClass=&NiagaraSpriteRendererProperties}
   {rendererClass=&NiagaraSpriteRendererProperties, rendererIndex=1}
   {rendererClass=&NiagaraSpriteRendererProperties, rendererIndex=2}
   {rendererClass=&NiagaraSpriteRendererProperties, rendererIndex=3}
   {rendererClass=&NiagaraSpriteRendererProperties, rendererIndex=4}
   {rendererClass=&NiagaraSpriteRendererProperties, rendererIndex=5}
   {rendererClass=&NiagaraSpriteRendererProperties, rendererIndex=6}
   {rendererClass=&NiagaraSpriteRendererProperties, rendererIndex=7}
  ]
 }
 Ingame_New {
  emitterSpawnScript [/SetVariables(Emitter.ParticlesPerTrail=5) {isSetParameters=true}]
  emitterUpdateScript [
   &EmitterState(
    Inactive Response=::NewEnumerator1
    Life Cycle Mode=::NewEnumerator1
    Loop Behavior=::NewEnumerator1
    Loop Duration Mode=::NewEnumerator1
    UseLoopDelay=false
   )
   &SpawnBurst_Instantaneous(
    Spawn Count=&Multiply_Int(A=$Emitter.ParticlesPerTrail, B=2)
    Use Spawn Probability=false
    UseLoopCountLimit=false
   )
   &SampleSkeletalMesh(
    AdditionalYawPitchRoll=false
    Bone Sampling Mode:ENiagaraBoneSamplingMode=::NewEnumerator4
    Skeletal Mesh:NiagaraDataInterfaceSkeletalMesh=$System.SkeletalMesh
   )
   &SampleSkeletalMesh(
    AdditionalYawPitchRoll=false
    Bone / Socket Index=1
    Bone Sampling Mode:ENiagaraBoneSamplingMode=::NewEnumerator4
    Skeletal Mesh:NiagaraDataInterfaceSkeletalMesh=$System.SkeletalMesh
   ) {module=SampleSkeletalMesh001}
   /SetVariables(Emitter.InterpolateRate:NiagaraFloat=&FloatFromCurve(FloatCurve=DataInterface() {propertyValues {Curve {keys [<at 0 50>,<at 300 3>]}}}, CurveIndex=$System.ForwardVelocity)) {isSetParameters=true}
   /SetVariables(Emitter.InterpolateTargetLerpAlpha:NiagaraFloat=&FloatFromCurve(
     FloatCurve=DataInterface() {propertyValues {Curve {keys [<at 0 0>,<at 350 1>]}}}
     CurveIndex=&VectorLength(Vector=$Engine.Owner.Velocity)
    )) {isSetParameters=true}
   /SetVariables(
    Emitter.InterpolateTargetL:Vector3f=&AddVectorToPosition(
     Position=$Emitter.SampleSkeletalMesh.SampledPosition
     Vector=&RotateVectorByQuaternion(Vector To Rotate=<-2.6198 -3.2948 -4.2354>, Quaternion=$Emitter.SampleSkeletalMesh.SampledRotation)
    )
    Emitter.InterpolateTargetR:Vector3f=&AddVectorToPosition(
     Position=$Emitter.SampleSkeletalMesh001.SampledPosition
     Vector=&RotateVectorByQuaternion(Vector To Rotate=<2.6198 3.2948 4.2354>, Quaternion=$Emitter.SampleSkeletalMesh001.SampledRotation)
    )
   ) {isSetParameters=true}
   /SetVariables(
    Emitter.InterpolateTargetStaticL:NiagaraPosition=&AddVectorToPosition(
     Position=Linked() {linkedVariable {name=Emitter.InterpolateTargetL, type=Vector3f}}
     Vector=&RotateVectorByQuaternion(
      Vector To Rotate=&Multiply_VectorByFloat(Vector=<-1 -100 3>, Float=$Emitter.InterpolateTargetLerpAlpha)
      Quaternion=$Engine.Owner.Rotation
     )
    )
    Emitter.InterpolateTargetStaticR:Vector3f=&AddVectorToPosition(
     Position=Linked() {linkedVariable {name=Emitter.InterpolateTargetR, type=Vector3f}}
     Vector=&RotateVectorByQuaternion(
      Vector To Rotate=&Multiply_VectorByFloat(Vector=<1 100 -5>, Float=$Emitter.InterpolateTargetLerpAlpha)
      Quaternion=$Engine.Owner.Rotation
     )
    )
   ) {isSetParameters=true}
   /SetVariables(
    Emitter.RibbonFacingL:Vector3f=&RotateVectorByQuaternion(Vector To Rotate=<0.2 0 -1>, Quaternion=$Emitter.SampleSkeletalMesh.SampledRotation)
    Emitter.RibbonFacingR:Vector3f=&RotateVectorByQuaternion(Vector To Rotate=<-0.2 0 1>, Quaternion=$Emitter.SampleSkeletalMesh001.SampledRotation)
   ) {isSetParameters=true}
  ]
  particleSpawnScript [
   &FN_SetSpawnGroup(Number of Spawn Groups=2, Set Ribbon ID and Ribbon Link Order=true)
   &PartitionParticles(
    Particles Per Partition=2
    Partition Count Mode:ENiagara_InfiniteLoopDuration=::NewEnumerator1
    Partition Mode:ENiagara_PartitionMode_v2=::NewEnumerator1
   )
   &InitializeParticle(
    Color Mode=::NewEnumerator0
    Lifetime=5
    Position=&MergeHelper_VectorSwitch(
     0=Linked() {linkedVariable {name=Emitter.SampleSkeletalMesh.SampledPosition, type=NiagaraPosition}}
     1=Linked() {linkedVariable {name=Emitter.SampleSkeletalMesh001.SampledPosition, type=NiagaraPosition}}
     2=?
     3=?
     Index=&SpawnGroup()
     Num=::NewEnumerator0
    )
    Position Mode=::NewEnumerator1
    Ribbon Width=&Multiply_Float(
     A=&MaskFloatBySpawnGroup(Masked Float=3, Passthrough Float=-3, Spawn Group Mask=1)
     B=&MergeHelper_FloatSwitch(
      0=&MakeFloatFromVector(Vector=$User.ShoeScale_L, Channel=::NewEnumerator0)
      1=&MakeFloatFromVector(Vector=$User.ShoeScale_R, Channel=::NewEnumerator0)
      2=?
      3=?
      Index=$Particles.SpawnGroup
      Num=::NewEnumerator0
     )
    )
    Ribbon Width Mode=::NewEnumerator1
    Sprite Size Mode=::NewEnumerator3
    Uniform Sprite Size=2
    UsePositionOffset=false
    Write Material Random=true
   )
   /SetVariables(Particles.ParticleLerpAlpha:NiagaraFloat=&Divide_Float(
     A=&MakeFloatFromInt(Int=$Particles.PartitionParticles.Partition)
     B=&MakeFloatFromInt(Int=$Emitter.ParticlesPerTrail)
    )) {isSetParameters=true}
   /SetVariables(
    Particles.RibbonID=/Niagara/DynamicInputs/NiagaraID/MakeNiagaraIDFromInt(Index:NiagaraInt32=&SpawnGroup(), Acquire Tag:NiagaraInt32=0)
    Particles.RibbonLinkOrder:NiagaraFloat=&ReturnNormalizedExecIndex()
   ) {isSetParameters=true}
   &CameraOffset(Camera Offset Amount=9)
   &KillParticles(Kill Particles=&SetBoolByFloatComparison(A=$Particles.RibbonWidth, B=0, Comparison Type=::NewEnumerator4))
  ]
  particleUpdateScript [
   /SetVariables(Particles.ParticleInterpolateRate:NiagaraFloat=&Lerp_Float(
     A=100
     B=$Emitter.InterpolateRate
     Alpha=&OneMinusFloat(Float=&Power(a=&OneMinusFloat(Float=$Particles.ParticleLerpAlpha), b=2))
    )) {isSetParameters=true}
   /SetVariables(Particles.InterpolateTarget:NiagaraPosition=&MakeCustomVectorFromBool(
     Bool=&Boolean_OrOperation(Bool1=$User.bIsFrontEnd, Bool2=$User.bIsFrontEndPreview)
     True Vector=&MergeHelper_VectorSwitch(
      0=$Emitter.InterpolateTargetL
      1=$Emitter.InterpolateTargetR
      2=?
      3=?
      Index=&SpawnGroup()
      Num=::NewEnumerator0
     )
     False Vector=&MergeHelper_VectorSwitch(
      0=$Emitter.InterpolateTargetL
      1=$Emitter.InterpolateTargetR
      2=?
      3=?
      Index=&SpawnGroup()
      Num=::NewEnumerator0
     )
    )) {isSetParameters=true}
   &ParticleState(Kill Particles When Lifetime Has Elapsed=false)
   &InterpolateOverTime(
    Rate Of Change:NiagaraFloat=$Particles.ParticleInterpolateRate
    Target Position Value:NiagaraPosition=$Particles.InterpolateTarget
    Variable Type=::NewEnumerator6
   )
   /SetVariables(
    Particles.Position=$Particles.InterpolateOverTime.InterpolateOverTime_Position.Moving Average
    Particles.PositionBlue:NiagaraPosition=&AddVectorToPosition(Position=$Particles.InterpolateOverTime.InterpolateOverTime_Position.Moving Average, Vector=<0 0 -3>)
    Particles.RibbonFacing=&MergeHelper_VectorSwitch(0=$Emitter.RibbonFacingL, 1=$Emitter.RibbonFacingR, 2=?, 3=?, Index=&SpawnGroup(), Num=::NewEnumerator0)
   ) {isSetParameters=true}
  ]
  rendererClasses [&NiagaraRibbonRendererProperties]
  renderers [
   {rendererClass=&NiagaraRibbonRendererProperties}
   {rendererClass=&NiagaraRibbonRendererProperties, rendererIndex=1}
  ]
 }
}
summary {
 systemName=NS_Shoe_HeatShineTorn
 emitters [
  {emitterName=FE_OffPawn, rendererClasses [&NiagaraSpriteRendererProperties]}
  {emitterName=Ingame_New, rendererClasses [&NiagaraRibbonRendererProperties]}
 ]
 userVariables [
  User.bIsFrontEnd {type=NiagaraBool}
  User.bIsFrontEndPreview {type=NiagaraBool}
  User.isOffPawnPreview {type=NiagaraBool}
  User.IsSkydiving {type=NiagaraBool}
  User.RibbonIDOffset {type=NiagaraInt32}
  User.ShoeScale_L {type=Vector3f, description="", defaultValue {struct=Vector3f, value=<1 1 1>}}
  User.ShoeScale_R {type=Vector3f, description="", defaultValue {struct=Vector3f, value=<1 1 1>}}
  User.SkeletalMesh {
   type=NiagaraDataInterfaceSkeletalMesh
   defaultValue {
    struct=NiagaraExt_VariableValue_DataInterface
    value {dataInterface=&self:NiagaraDataInterfaceSkeletalMesh_0}
   }
  }
 ]
}
