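# Parent material for the glass on the HeatShineTorn cosmetic shoe, faking see-through depth and glow.
# Bump offset plus normal-map distortion warps the emissive UVs fed to the character PreFX function.
# Base color adds an inner-shoe texture and hue-shifted tints; emissive is masked, hue-shifted and scaled by left/right shoe parameters.
# Boosted specular and a clear-coat bottom normal pass through MF_Sock into PostFX; a sine blink branch is left unconnected.
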Material /CosmeticShoes/Assets/FORT_Shoes/Shoe_HeatShineTorn/Materials/M_HeatShineTorn_Shoe_Glass_Parent
graph {
 properties {
  bUseMaterialAttributes=true
  bUsedWithClothing=true
  bUsedWithMeshDeformer=true
  bUsedWithMorphTargets=true
  bUsedWithNanite=true
  bUsedWithNiagaraMeshParticles=true
  bUsedWithSkeletalMesh=true
  pixelDepthOffsetMode=PDOM_Legacy
  shadingModel=MSM_FromMaterialExpression
  usedShadingModels=Default Lit | Subsurface | Clear Coat | Subsurface Profile | Cloth
 }
 nodes [
  n1 = Constant2Vector() {g=0.01}
  n2 = NamedRerouteUsage()
  n3 = OneMinus(None=@n2)
  n4 = BumpOffset(
   Coordinate=Add(A=TextureCoordinate(), B=Multiply(A=@n1, B=@n3))
   Height=LinearInterpolate(A=ScalarParameter() {defaultValue=-0.3, parameterName=HeatShineTorn_BumpOffset}, Alpha=@n2) {constB=0.25}
  )
  n5 = TextureSample() {
   samplerSource=SSM_Wrap_WorldGroupSettings
   samplerType=SAMPLERTYPE_Normal
   texture=/CosmeticShoes/Assets/FORT_Shoes/Shoe_HeatShineTorn/Textures/T_HeatShineTorn_Shoe_N
  }
  n6 = Multiply(
   A=ComponentMask(None=NamedRerouteDeclaration(None=@n5.RGB) {name=TexturedNormal, nodeColor=<rgba 1 0.7969 0 1>}) {g=true, r=true}
   B=ScalarParameter() {defaultValue=0.007, parameterName=HeatShineTorn_NormalDistortion}
  ) {constB=0}
  n7 = MaterialFunctionCall(Override Basic Emissive UVs=NamedRerouteDeclaration(None=Add(A=@n4, B=@n6)) {name=EmissiveUVs, nodeColor=<rgba 0.0547 0 1 1>}) {materialFunction=&MF_CharacterParent_2023_PreFX}
  n8 = GetMaterialAttributes(None=@n7.Result) {attributeGetTypes [#1, #2, #3]}
  n9 = ScalarParameter() {defaultValue=0.1, parameterName=HeatShineTorn_SpecularOffset}
  n10 = SetMaterialAttributes!(
   MaterialAttributes=@n8.MaterialAttributes
   Emissive Color=NamedRerouteUsage()
   Specular=Saturate(None=Add(A=@n8.Specular, B=@n9))
   ClearCoatBottomNormal=NamedRerouteUsage()
   Base Color=NamedRerouteUsage()
  ) {attributeSetTypes [
    #1
    #2
    #4
    #3
   ]}
  MaterialFunctionCall(
   Material=MaterialFunctionCall!(InOpacity, Material=@n10, Mask=@n7.Tech Art Mask) {materialFunction=/Game/Global/Materials/Cosmetics/Kicks/Socks/Materials/MF_Sock}.Result
   M=@n7.M
   Pre Skinned Local Position=@n7.M
   Distance Blend=@n7.M
   Fresnel=@n7.M
  ) {materialFunction=&MF_CharacterParent_2023_PostFX}
  n11 = ScalarParameter() {parameterName=HeatShineTorn_UseEmissive_R}
  n12 = Constant() {r=0.8}
  n13 = Constant() {r=0.8}
  n14 = ScalarParameter() {parameterName=HeatShineTorn_UseEmissive_L}
  n15 = LinearInterpolate(A=Add(A=Multiply(A=@n11, B=@n12), B=@n13), B=Add(A=Multiply(A=@n14, B=@n12), B=@n13), Alpha=NamedRerouteUsage())
  n16 = Sine(None=Multiply(A=Time()) {constB=0.2})
  n17 = Constant() {r=1}
  Multiply(A=@n15, B=If(A=Abs(None=@n16), "A > B"=@n17, "A == B"=@n17, "A < B"=Constant()) {constB=0.6})
  n18 = NamedRerouteUsage()
  n19 = Add(A=ComponentMask(None=@n18) {b=true}) {constB=-4.2}
  NamedRerouteDeclaration(None=Saturate(None=Multiply(A=@n19) {constB=100})) {name=UpperGlassMask, nodeColor=<rgba 1 0.3984 0 1>}
  NamedRerouteDeclaration(None=@n8.Emissive Color) {name=OriginalEmissive, nodeColor=<rgba 1 0 0 1>}
  NamedRerouteDeclaration(None=@n8.Base Color) {name=OriginalBaseColor, nodeColor=<rgba 0 1 0.2969 1>}
  n20 = NamedRerouteUsage()
  n21 = ComponentMask(None=@n18) {b=true}
  n22 = Multiply(A=NamedRerouteUsage(), B=Saturate(None=Add(A=@n21) {constB=-0.6}))
  n23 = Desaturation(None=@n22, Fraction=Constant() {r=1})
  n24 = Multiply(A=Power(Base=Saturate(None=@n23)) {constExponent=4}) {constB=100}
  n25 = TextureSample(UVs=NamedRerouteUsage()) {
   samplerSource=SSM_Clamp_WorldGroupSettings
   texture=/CosmeticShoes/Assets/FORT_Shoes/Shoe_HeatShineTorn/Textures/T_HeatShineTorn_Shoe_Inner_E_Shrunk
  }
  n26 = Constant() {r=0.35}
  n27 = LinearInterpolate(
   A=Multiply(A=Saturate(None=@n24)) {constB=0.15}
   B=Multiply(A=Desaturation(None=@n25.RGB, Fraction=@n26)) {constB=0.75}
   Alpha=NamedRerouteUsage()
  )
  n28 = NamedRerouteUsage()
  n29 = Desaturation(
   None=MaterialFunctionCall(Hue Shift Percentage=ScalarParameter() {defaultValue=0.05, parameterName=aa}, Texture=@n28) {materialFunction=&HueShift}.Result
   Fraction=Constant() {r=-1}
  )
  n30 = NamedRerouteUsage()
  NamedRerouteDeclaration(None=Add(A=Add(A=@n20, B=@n27), B=LinearInterpolate(A=@n28, B=@n29, Alpha=@n30))) {name=OutputBaseColor, nodeColor=<rgba 0.6406 1 0 1>}
  NamedRerouteDeclaration(None=@n7.Pre skinned Local Position) {name=PreskinnedLocalPosition, nodeColor=<rgba 1 0 0.9609 1>}
  n31 = Constant() {r=-0.05}
  n32 = LinearInterpolate(Alpha=@n15) {constB=0.5}
  n33 = Add(A=OneMinus(None=NamedRerouteUsage()), B=NamedRerouteUsage())
  n34 = Desaturation(None=Multiply(A=Multiply(A=@n22, B=@n32), B=Saturate(None=@n33)), Fraction=Constant() {r=-1})
  NamedRerouteDeclaration(None=Multiply(A=MaterialFunctionCall(Hue Shift Percentage=@n31, Texture=@n34) {materialFunction=&HueShift}.Result) {constB=0.5}) {name=OutputEmissive, nodeColor=<rgba 1 0.7969 0 1>}
  NamedRerouteDeclaration(None=@n7.Fresnel) {name=Fresnel, nodeColor=<rgba 0.7344 0 1 1>}
  n35 = NamedRerouteUsage()
  NamedRerouteDeclaration(None=Saturate(None=ComponentMask(None=@n35) {r=true})) {name=Left/Right_Mask, nodeColor=<rgba 1 0.1641 0 1>}
 ]
}
