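# Parent material for the HeatShineTorn cosmetic shoe, built on the shared 2023 character PreFX and PostFX functions.
# A saturated pre-skinned local X mask blends separate right and left logo parameters into a multiplier on the PreFX emissive, giving per-foot control.
# The result passes through shared shoe reactivity and sock functions, then PostFX, and finally the tech-art shoe hologram function.
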Material /CosmeticShoes/Assets/FORT_Shoes/Shoe_HeatShineTorn/Materials/M_HeatShineTorn_Shoe_Parent
graph {
 properties {
  bUseMaterialAttributes=true
  bUsedWithClothing=true
  bUsedWithMeshDeformer=true
  bUsedWithMorphTargets=true
  bUsedWithNanite=true
  bUsedWithNiagaraMeshParticles=true
  bUsedWithSkeletalMesh=true
  pixelDepthOffsetMode=PDOM_Legacy
  shadingModel=MSM_FromMaterialExpression
  usedShadingModels=Default Lit | Subsurface | Clear Coat | Subsurface Profile | Cloth
 }
 nodes [
  n1 = MaterialFunctionCall() {materialFunction=&MF_CharacterParent_2023_PreFX}
  n2 = GetMaterialAttributes(None=@n1.Result) {attributeGetTypes [#1]}
  n3 = Constant() {r=0.8}
  n4 = ScalarParameter() {parameterName=HeatShineTorn_Logo_UseEmissive_R}
  n5 = Constant() {r=0.8}
  n6 = ScalarParameter() {parameterName=HeatShineTorn_Logo_UseEmissive_L}
  n7 = LinearInterpolate(A=Add(A=Multiply(A=@n3, B=@n4), B=@n5), B=Add(A=Multiply(A=@n3, B=@n6), B=@n5), Alpha=NamedRerouteUsage())
  n8 = MaterialFunctionCall!(Material=SetMaterialAttributes(MaterialAttributes=@n2.MaterialAttributes, Emissive Color=Multiply(A=@n2.Emissive Color, B=@n7)) {attributeSetTypes [#1]}) {materialFunction=/Game/Global/Materials/Cosmetics/Kicks/Shoes/Materials/MF_ShoeReactivityTest}
  MaterialFunctionCall!(In=MaterialFunctionCall(
    Material=MaterialFunctionCall!(InOpacity, Material=@n8.Result, Mask=@n1.Tech Art Mask) {materialFunction=/Game/Global/Materials/Cosmetics/Kicks/Socks/Materials/MF_Sock}.Result
    M=@n1.M
    Pre Skinned Local Position=@n1.M
    Distance Blend=@n1.M
    Fresnel=@n1.M
   ) {materialFunction=&MF_CharacterParent_2023_PostFX}.Result) {materialFunction=/Game/Global/Materials/Cosmetics/Kicks/Shoes/Materials/MF_TechartShoeHologram}
  NamedRerouteDeclaration(None=Saturate(None=ComponentMask(None=@n1.Pre skinned Local Position) {r=true})) {name=Left/Right_Mask, nodeColor=<rgba 1 0.1641 0 1>}
 ]
}
