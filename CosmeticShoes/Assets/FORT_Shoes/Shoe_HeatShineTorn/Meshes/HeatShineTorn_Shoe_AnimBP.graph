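# Animation Blueprint for the HeatShineTorn cosmetic shoes, animating them and their glow during the matching emote.
# The shoes copy the attached parent's pose.
# When the pawn's current montage name contains "Emote_HeatShineTorn", a ShoeBlend mask layers in a single-state, non-looping shoe animation.
# Each update refreshes that flag and feeds animation curves into left and right emissive parameters on the shoe and logo dynamic materials.
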AnimBlueprint /CosmeticShoes/Assets/FORT_Shoes/Shoe_HeatShineTorn/Meshes/HeatShineTorn_Shoe_AnimBP
target_skeleton=&/HeatShineTorn_Shoe_Skeleton
graphs [
 AnimGraph {nodes [
   AnimGraphNode_StateMachine_3 = Emote_HeatShineTorn_Shoes_Equip() {is=title AnimGraphNode_StateMachine}
   n1 = Layered blend per bone(BasePose=Copy Pose From Mesh() {bUseAttachedParent=true}, BlendPoses_0=@AnimGraphNode_StateMachine_3) {
    blendMode=BlendMask
    blendMasks [&HeatShineTorn_Shoe_Skeleton:ShoeBlend]
    layerSetup []
   }
   Output Pose(Result=Blend Poses by bool(
     BlendPose_0=@n1
     BlendPose_1=Copy Pose From Mesh() {bUseAttachedParent=true}
     bActiveValue=Get HeatShineTorn Is Playing() -> (HeatShineTorn Is Playing:bool)
    ))
  ]}
 Emote_HeatShineTorn_Shoes_Equip {
  entry=Emote_HeatShineTorn_Shoes
  path=AnimGraph.AnimGraphNode_StateMachine_3.Emote_HeatShineTorn_Shoes_Equip
  ids=id
  nodes [AnimStateNode_3 = Emote_HeatShineTorn_Shoes() {
    is=title AnimStateNode
    stateEntered {guid=#1}
    stateFullyBlended {guid=#2}
    stateLeft {guid=#3}
   }]
 }
 Emote_HeatShineTorn_Shoes {
  path=AnimGraph.AnimGraphNode_StateMachine_3.Emote_HeatShineTorn_Shoes_Equip.AnimStateNode_3.Emote_HeatShineTorn_Shoes
  nodes [
   n1 = Sequence Player 'Emote_HeatShineTorn_Shoes'() {bLoopAnimation=false, sequence=/CosmeticShoes/Assets/FORT_Shoes/Shoe_HeatShineTorn/Meshes/HeatShineTorn_Emote/Emote_HeatShineTorn_Shoes}
   Output Animation Pose(Result=@n1)
  ]
 }
 EventGraph {nodes [
   Event BlueprintUpdateAnimation()
   > n1 = Sequence!(execute) -> (then_0, then_1, then_2:Exec, then_3:Exec, then_4:Exec)
   > n2 = Get Dynamic Material Instance(SkeletalMeshComponent=GetPartSkeletalMeshComponent())
   > SetScalarParameterValue(self=@n2, ParameterName=HeatShineTorn_UseEmissive_L, Value=GetCurveValue(CurveName=HeatShineTorn_UseEmissive_L))
   n3 = Get Dynamic Material Instance(execute=@n1.then_1, SkeletalMeshComponent=GetPartSkeletalMeshComponent(), ElementIndex=1)
   > SetScalarParameterValue(self=@n3, ParameterName=HeatShineTorn_Logo_UseEmissive_L, Value=GetCurveValue(CurveName=HeatShineTorn_Logo_UseEmissive_L))
   n4 = Get Dynamic Material Instance(execute=@n1.then_2, SkeletalMeshComponent=GetPartSkeletalMeshComponent())
   > SetScalarParameterValue(self=@n4, ParameterName=HeatShineTorn_UseEmissive_R, Value=GetCurveValue(CurveName=HeatShineTorn_UseEmissive_R))
   n5 = Get Dynamic Material Instance(execute=@n1.then_3, SkeletalMeshComponent=GetPartSkeletalMeshComponent(), ElementIndex=1)
   > SetScalarParameterValue(self=@n5, ParameterName=HeatShineTorn_Logo_UseEmissive_R, Value=GetCurveValue(CurveName=HeatShineTorn_Logo_UseEmissive_R))
   n6 = GetCurrentMontage!(self:Character Object Reference=Get FortPlayerPawn() -> (FortPlayerPawn:Fort Player Pawn Object Reference)) -> (ReturnValue:Anim Montage Object Reference)
   Set HeatShineTorn Is Playing(execute=@n1.then_4, HeatShineTorn Is Playing:bool=Contains(SearchIn=GetObjectName(Object=@n6), Substring=Emote_HeatShineTorn)) -> (Output_Get:bool)
   Event BlueprintInitializeAnimation()
   > n7 = Cast To FortPlayerPawn(Object=TryGetPawnOwner())
   > Set FortPlayerPawn(FortPlayerPawn:Fort Player Pawn Object Reference=@n7) -> (Output_Get:Fort Player Pawn Object Reference)
   n8 = Get FortPlayerPawn()
   > n9 = Cast To Fortnite_M_Avg_Player_AnimBlueprint!(execute, Object=GetAnimInstance(self=Get Mesh(self=@n8))) -> (then, CastFailed, AsFortnite M Avg Player Anim Blueprint:Fortnite M Avg Player Anim Blueprint Object Reference) {is=title K2Node_DynamicCast}
   > Set FortPlayerAnimBP(FortPlayerAnimBP:Fortnite M Avg Player Anim Blueprint Object Reference=@n9) -> (Output_Get:Fortnite M Avg Player Anim Blueprint Object Reference)
   TryGetPawnOwner()
   Set HeatShineTorn Is Playing(HeatShineTorn Is Playing=false)
  ]}
 OnStateEntry {nodes [
   OnStateEntry!() -> (then, Context:"Anim Update Context Structure (by ref)", Node:"Anim Node Reference Structure (by ref)") {is=title K2Node_FunctionEntry}
   > Return Node()
  ]}
]
variables [
 FortPlayerAnimBP {type=Fortnite_M_Avg_Player_AnimBlueprint_C, default=null}
 FortPlayerPawn {type=FortPlayerPawn}
 HeatShineTorn Is Playing {type=boolean}
 HeatShineTorn_Emote {type=AnimMontage, default=/BRCosmetics/Animation/Game/MainPlayer/Emotes/HeatShineTorn/CMM/Emote_HeatShineTorn_CMM_Equip_M}
 NewVar {type=number}
]
